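; Animation modifier for the dog pet that strips baked scarf-tie animation from an Anim Sequence.
; On apply, it removes the bone tracks for the left and right scarf tie bones, dyn_scarf_tie_l and dyn_scarf_tie_r, using the Animation Blueprint Library.
; The file defines no revert logic to restore the removed tracks.
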(event EventOnApply
  (params (AnimationSequence :type Anim%20Sequence%20Object%20Reference))
  (AnimationBlueprintLibrary|Bones|RemoveBoneAnimation AnimationSequence "dyn_scarf_tie_l")
  (AnimationBlueprintLibrary|Bones|RemoveBoneAnimation AnimationSequence "dyn_scarf_tie_r"))
